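# Translucent unlit Niagara material for a ground ring of dust in the grapple knife's effects.
# A panned smoke texture, with its pan time driven by the ExplosionPan dynamic parameter, passes through a SmoothStep with Min and Max parameters.
# A U-axis gradient mask fades both edges, then vertex color, particle alpha and depth fade shape the opacity.
# Particle color feeds a shared emissive adjustment function.
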Material /GrappleWeapon/Items/GrappleKnife/FX/Materials/M_Ground_Ring_Dust
graph {
 properties {
  bUsedWithNiagaraMeshParticles=true
  bUsedWithNiagaraSprites=true
  blendMode=BLEND_Translucent
  pixelDepthOffsetMode=PDOM_Legacy
  shadingModel=MSM_Unlit
  usedShadingModels=Unlit
 }
 nodes [
  n1 = Panner(
   Coordinate=TextureCoordinate() {uTiling=0.75, vTiling=0.8}
   Time=DynamicParameter() {defaultValue=<rgba -1 1 1 1>, paramNames [ExplosionPan,Param2,Param3,Param4]}.ExplosionPan
  ) {speedY=-0.25}
  n2 = ComponentMask(None=TextureCoordinate()) {r=true}
  n3 = Multiply(
   A=MaterialFunctionCall(
    Alpha=TextureSample(UVs=@n1) {samplerType=SAMPLERTYPE_LinearGrayscale, texture=&/T_Smoke_Wind_Dissolve_Handpainted}.G
    Min=ScalarParameter() {parameterName=Min}
    Max=ScalarParameter() {defaultValue=2, parameterName=Max}
   ) {materialFunction=&SmoothStep}.Result
   B=Multiply(A=@n2, B=OneMinus(None=@n2))
  )
  n4 = VertexColor()
  DepthFade(Opacity=Multiply(A=ParticleColor().A, B=Multiply(A=@n3, B=@n4)), FadeDistance=ScalarParameter() {defaultValue=50, parameterName=DepthFade})
  MaterialFunctionCall(Emissive=ParticleColor().RGB) {materialFunction=&MF_FX_EmissiveAdjustments}
 ]
}
