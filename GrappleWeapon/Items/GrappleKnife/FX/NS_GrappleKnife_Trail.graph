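# Niagara swing trail for the GrappleKnife weapon, with extra effects for its 360 attack.
# Two ribbons, a soft AnimTrail and a wider trail sampled from the skeletal mesh, fade and narrow over their short lifetimes.
# User.is360Atk changes ribbon lifetimes, makes them inherit ParentVelocity, and enables an inner flash sprite plus rotating wind-streak mesh bursts.
# Every emitter tints by Fortnite time of day through a local ColorFromTOD_02 module.
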NiagaraSystem /GrappleWeapon/Items/GrappleKnife/FX/NS_GrappleKnife_Trail
dependencies {
 usedDataInterfaces [
  &NiagaraDataInterfaceCurve
  &NiagaraDataInterfaceColorCurve
  &NiagaraDataInterfaceArrayFloat3
  &NiagaraDataInterfaceArrayFloat
  &NiagaraDataInterfaceVector2DCurve
  &NiagaraDataInterfaceSkeletalMesh
 ]
 usedDynamicInputs []
 usedModules [
  &self:SystemSpawnScript.SystemScriptSource.SystemScriptGraph.NiagaraNodeAssignment_0.SetVariables_#1
  &SystemState
 ]
}
emitters {
 Facing_Ribbon_Width {
  emitterUpdateScript [
   &EmitterState(Life Cycle Mode=::NewEnumerator1, Loop Behavior=::NewEnumerator1, Loop Duration=1, UseLoopDelay=true)
   &SpawnRate(SpawnRate=40, Use Spawn Probability=false)
  ]
  particleSpawnScript [
   &SkeletalMeshLocation(
    MaskSpawnGroup=false
    Mesh Sampling Type=::NewEnumerator5
    Sampled Position Offset=<0 0 40>
    Skeletal Mesh=$System.Skeletal Mesh
    Socket Sampling Mode=::NewEnumerator5
   )
   &InitializeRibbon(
    Color=<rgba 1 0.9437 0.8958 0.75>
    Lifetime=&MakeCustomFloatFromBool(Bool=$User.is360Atk, True Float=0.1, False Float=0.15)
    Module.Write Color=true
    Module.Write Lifetime=true
    Module.Write Mass=true
    Module.Write Material Random=true
    Module.Write Position=false
    Module.Write RibbonTwist=false
    Module.Write RibbonWidth=true
    Ribbon Width=40
   )
   &CameraOffset(Camera Offset Amount=5)
  ]
  particleUpdateScript [
   &DynamicMaterialParameters(
    Index 0 Param 1=2
    Module.Param0WriteEnabled=true
    Module.Param1WriteEnabled=true
    Module.Param2WriteEnabled=true
    Module.Param3WriteEnabled=true
   )
   &ParticleState()
   &InheritSourceMovement(
    Applied Position Scale=&MakeCustomVectorFromBool(Bool=$User.is360Atk, True Vector=<1 1 1>, False Vector=<0 0 0>)
    ApplyForce=false
    ApplyPosition=true
    ApplyPositionFalloff=false
    Source Velocity:Vector3f=$User.ParentVelocity
   )
   &SolveForcesAndVelocity(Clamp Velocity=false, Limit Acceleration=false)
   &ScaleColor(
    Scale Alpha=&Multiply_Float(
     A=&FloatFromCurve(FloatCurve=DataInterface() {propertyValues {Curve {keys [
          {interpMode=RCIM_Cubic, leaveTangent=-2.3902, tangentMode=RCTM_Break, time=0, value=1}
          {arriveTangent=-0.1715, interpMode=RCIM_Cubic, tangentMode=RCTM_Break, time=1, value=0}
         ]}}})
     B=&Multiply_Float(A=1, B=&FNTimeOfDayScalar(Night Scale=1, Morning Scale=0.5, Evening Scale=0.5))
    )
    Scale RGB=&VectorFromFloat(Value=&FloatFromCurve(FloatCurve=DataInterface() {propertyValues {Curve {keys [
          <at -0.0016 1>
          {
           arriveTangent=-0.3333
           arriveTangentWeight=0.131
           interpMode=RCIM_Cubic
           leaveTangent=-0.3333
           leaveTangentWeight=0.0674
           tangentMode=RCTM_User
           time=0.7977
           value=0.75
          }
         ]}}}))
    ScaleA=true
    ScaleRGB=true
   )
   &ScaleRibbonWidth(Ribbon Width Scale=&FloatFromCurve(FloatCurve=DataInterface() {propertyValues {Curve {keys [<at 0 1.5 RCIM_Cubic>, {arriveTangent=-0.2362, interpMode=RCIM_Cubic, leaveTangent=-0.2362, tangentMode=RCTM_User, time=1, value=0.1}]}}})) {module=ScaleRibbonWidth001}
   &self:ColorFromTOD_02(
    Afternoon:LinearColor=<rgba 1 0.9711 0.7999 0.9>
    Evening:LinearColor=<rgba 1 0.625 0.5 1>
    Morning:LinearColor=<rgba 1 0.75 0.7753 1>
    Night:LinearColor=<rgba 0.7 0.3494 0.4483 1>
    TOD:Vector4f=&MakeVector4_FromLinearColor(LinearColor=$NPC.FortniteNPC.FortniteActiveTimeOfDay)
   ) {module=ColorFromTOD_02}
  ]
  rendererClasses [&NiagaraRibbonRendererProperties]
  renderers [{rendererClass=&NiagaraRibbonRendererProperties}]
 }
 InnerFlash {
  emitterUpdateScript [
   &EmitterState(Life Cycle Mode=::NewEnumerator1, Loop Behavior=::NewEnumerator1, Loop Duration=1, UseLoopDelay=false)
   &SpawnBurst_Instantaneous(
    Spawn Probability=&MakeCustomFloatFromBool(Bool=$User.is360Atk, True Float=1, False Float=0)
    Spawn Time=0.08
    Use Spawn Probability=true
    UseLoopCountLimit=false
   )
  ]
  particleSpawnScript [
   &InitializeParticle(
    Color=<rgba 1 1 1 0.5>
    Lifetime=&Multiply_Float(A=0.4, B=$User.LifetimeScale)
    Sprite Size Mode=::NewEnumerator3
    Uniform Sprite Size=150
    UsePositionOffset=false
    Write Material Random=true
   )
   &CameraOffset(Camera Offset Amount=25)
  ]
  particleUpdateScript [
   &ParticleState()
   &ScaleSpriteSize(
    Uniform Curve Scale=&Multiply_Float(A=1, B=$User.ScaleSize)
    Uniform Curve Sprite Scale=DataInterface() {propertyValues {Curve {keys [
        {arriveTangent=9.0725, interpMode=RCIM_Cubic, leaveTangent=9.0725, tangentMode=RCTM_User, time=0, value=0}
        {arriveTangent=-2.5249, interpMode=RCIM_Cubic, leaveTangent=-2.5249, tangentMode=RCTM_User, time=0.1408, value=0.9847}
        {arriveTangent=-0.0405, interpMode=RCIM_Cubic, leaveTangent=-0.0405, tangentMode=RCTM_User, time=1, value=0.45}
       ]}}}
   )
   &ScaleColor(
    Scale RGB=&MakeVectorFromLinearColorRGB(Color=&LerpLinearColors(
      StartColor=&ColorFromCurve(ColorCurve=DataInterface() {propertyValues {
         AlphaCurve {keys [<at 0 1>,<at 1 0>]}
         BlueCurve {keys [<at 0 1>,<at 0.5933 1>,<at 1 0.0577>]}
         GreenCurve {keys [<at 0 0.5013>,<at 0.5933 0.4172>,<at 1 1.16>]}
         RedCurve {keys [<at 0 0.156>,<at 0.5933 0.2699>,<at 1 3>]}
        }})
      EndColor=$User.ColorInner
      LerpFactor=&MakeCustomFloatFromBool(Bool=$User.SetColor, True Float=1, False Float=0)
     ))
    ScaleA=true
    ScaleRGB=true
   ) {module=ScaleColor001}
   &ScaleColor(
    Scale Alpha=&FloatFromCurve(
     FloatCurve=DataInterface() {propertyValues {Curve {keys [
         {arriveTangent=-0.5632, interpMode=RCIM_Cubic, leaveTangent=-0.5632, tangentMode=RCTM_User, time=0, value=1}
         {arriveTangent=-1, interpMode=RCIM_Cubic, leaveTangent=-1, time=0.2, value=0.4}
         <at 1 0>
        ]}}}
     Scale Curve=&Multiply_Float(A=0.3, B=$User.ScaleOpacity)
    )
    Scale RGB=&VectorFromFloat(Value=&FloatFromCurve(
      FloatCurve=DataInterface() {propertyValues {Curve {keys [
          {interpMode=RCIM_Cubic, tangentMode=RCTM_User, time=0, value=1}
          {arriveTangent=-1, interpMode=RCIM_Cubic, leaveTangent=-1, time=0.2, value=0.4}
          {interpMode=RCIM_Cubic, tangentMode=RCTM_User, time=1, value=0}
         ]}}}
      Scale Curve=0.3
     ))
    ScaleA=true
    ScaleRGB=true
   )
   &self:ColorFromTOD_02(
    Evening:LinearColor=<rgba 1 0.625 0.5 1>
    Morning:LinearColor=<rgba 1 0.75 0.7753 1>
    Night:LinearColor=<rgba 0.7 0.2966 0.3384 0.75>
    TOD:Vector4f=&MakeVector4_FromLinearColor(LinearColor=$NPC.FortniteNPC.FortniteActiveTimeOfDay)
   ) {module=ColorFromTOD_02}
  ]
  rendererClasses [&NiagaraSpriteRendererProperties]
  renderers [{rendererClass=&NiagaraSpriteRendererProperties}]
 }
 Main_Soft_Trail {
  emitterUpdateScript [
   &EmitterState(Life Cycle Mode=::NewEnumerator1, Loop Behavior=::NewEnumerator1, Loop Duration=1, UseLoopDelay=true)
   &FN_SpawnAnimTrail_v2()
  ]
  particleSpawnScript [
   /Game/Effects/Niagara/Modules/AnimTrail/FN_InitializeAnimTrail/FN_InitializeAnimTrailBoneIndecies(Second Bone or Socket Index=1)
   &FN_InitializeAnimTrail(
    AnimTrailWidth=3
    Color=<rgba 0.65 0.5552 0.4875 2.5>
    Lifetime=&MakeCustomFloatFromBool(Bool=$User.is360Atk, True Float=0.25, False Float=0.175)
    Position Offset=<0 0 -10>
    SkeletalMesh=$System.Skeletal Mesh
   ) {module=FN_InitializeAnimTrail001}
  ]
  particleUpdateScript [
   &FN_UpdateAnimTrail(SkeletalMesh=$System.Skeletal Mesh) {module=FN_UpdateAnimTrail001}
   &ScaleRibbonWidth(Ribbon Width Scale=&FloatFromCurve(
     FloatCurve=DataInterface() {propertyValues {Curve {keys [
         {interpMode=RCIM_Cubic, tangentMode=RCTM_User, time=0, value=1}
         {arriveTangent=-1.5591, interpMode=RCIM_Cubic, leaveTangent=-1.5591, tangentMode=RCTM_User, time=1, value=0}
        ]}}}
     Scale Curve=&MakeCustomFloatFromBool(Bool=$User.is360Atk, True Float=0.75, False Float=1)
    ))
   &InheritSourceMovement(
    Applied Position Scale=&MakeCustomVectorFromBool(Bool=$User.is360Atk, True Vector=<1 1 1>, False Vector=<0 0 0>)
    ApplyForce=false
    ApplyPosition=true
    ApplyPositionFalloff=false
    Source Velocity:Vector3f=$User.ParentVelocity
   )
   &SolveForcesAndVelocity(Clamp Velocity=false, Limit Acceleration=false)
   &ScaleColor(
    Scale Alpha=&FloatFromCurve(
     FloatCurve=DataInterface() {propertyValues {Curve {keys [
         {arriveTangent=-4.2552, interpMode=RCIM_Cubic, leaveTangent=-4.2552, tangentMode=RCTM_User, time=0, value=1}
         {arriveTangent=-0.4121, interpMode=RCIM_Cubic, leaveTangent=-0.4121, tangentMode=RCTM_User, time=0.4169, value=0.1059}
         {arriveTangent=-0.0022, interpMode=RCIM_Cubic, leaveTangent=-0.0022, tangentMode=RCTM_User, time=1, value=0}
        ]}}}
     Scale Curve=&FloatFromCurve(FloatCurve=DataInterface() {propertyValues {Curve {keys [<at 0 0>,<at 0.05 0>,<at 0.075 1>,<at 1 0>]}}}, CurveIndex=$System.Age)
    )
    Scale RGB=&VectorFromFloat(Value=&FloatFromCurve(FloatCurve=DataInterface() {propertyValues {Curve {keys [
          <at -0.0016 0.5202>
          {arriveTangentWeight=0.0023, interpMode=RCIM_Cubic, leaveTangentWeight=0.331, tangentMode=RCTM_User, time=0.0248, value=0.4661}
          {
           arriveTangent=-1.3798
           arriveTangentWeight=0.0174
           interpMode=RCIM_Cubic
           leaveTangent=-1.3798
           leaveTangentWeight=0.3136
           tangentMode=RCTM_User
           time=0.077
           value=0.3939
          }
          {
           arriveTangent=-0.7276
           arriveTangentWeight=0.1152
           interpMode=RCIM_Cubic
           leaveTangent=-0.7276
           leaveTangentWeight=0.1984
           tangentMode=RCTM_User
           time=0.4047
           value=0.0917
          }
          {
           arriveTangent=-0.3333
           arriveTangentWeight=0.131
           interpMode=RCIM_Cubic
           leaveTangent=-0.3333
           leaveTangentWeight=0.0674
           tangentMode=RCTM_User
           time=0.7977
           value=0
          }
         ]}}}))
    ScaleA=true
    ScaleRGB=true
   )
   &self:ColorFromTOD_02(
    Evening:LinearColor=<rgba 1 0.625 0.5 1>
    Morning:LinearColor=<rgba 1 0.75 0.7753 1>
    Night:LinearColor=<rgba 0.7 0.3494 0.4483 1>
    TOD:Vector4f=&MakeVector4_FromLinearColor(LinearColor=$NPC.FortniteNPC.FortniteActiveTimeOfDay)
   ) {module=ColorFromTOD_02}
  ]
  rendererClasses [&NiagaraRibbonRendererProperties]
  renderers [{rendererClass=&NiagaraRibbonRendererProperties}]
 }
 Wind_Streaks_Circle {
  emitterUpdateScript [
   &EmitterState(Life Cycle Mode=::NewEnumerator1, Loop Behavior=::NewEnumerator1, Loop Duration=0.5, UseLoopDelay=false)
   &SpawnBurst_Instantaneous(
    Spawn Count=4
    Spawn Probability=&MakeCustomFloatFromBool(Bool=$User.is360Atk, True Float=1, False Float=0)
    Use Spawn Probability=true
    UseLoopCountLimit=false
   )
  ]
  particleSpawnScript [
   &FN_SetSpawnGroup(Number of Spawn Groups=2)
   &InitializeParticle(
    Color Maximum=<rgba 1 0.7535 0.625 1.5>
    Color Minimum=<rgba 1 0.9541 0.9017 1>
    Color Mode=::NewEnumerator2
    Lifetime=0.5
    Mesh Scale=&Multiply_VectorByFloat(
     Vector=&SelectVectorFromArray(
      Direct Array Index=$Particles.SpawnGroup
      Vector Selection Array=DataInterface() {propertyValues {FloatData [<11.25 11.25 3>, <24 24 5>, <-15 -15 3.5>, <17 17 4.5>]}}
     )
     Float=0.2
    )
    Mesh Scale Mode=::NewEnumerator4
    Position Offset=<-50 -50 0>
    Position Offset Coordinate Space=::Local
    UsePositionOffset=true
    Write Material Random=true
   )
   &InitialMeshOrientation(
    Mesh Axis To Orient=<0 0 1>
    Mesh Orientation Mode=::NewEnumerator2
    Orientation Coordinate Space=::Local
    Rotation=&RandomRangeVector(Minimum=<-0.05 -0.05 -0.05>, Maximum=<-0.05 -0.05 0.05>)
    Rotation Coordinate Space=::NewEnumerator2
    Use Rotation Vector=true
   )
   &CameraOffset(Camera Offset Amount=&SelectFloatFromArray(Direct Array Index=$Particles.SpawnGroup, Float Selection Array=DataInterface() {propertyValues {FloatData [100,25]}}))
   &AddVelocity(
    Velocity=<0 0 5>
    Velocity Speed Scale=&SelectFloatFromArray(Direct Array Index=$Particles.SpawnGroup, Float Selection Array=DataInterface() {propertyValues {FloatData [0,1,0.25,0.5]}})
   )
  ]
  particleUpdateScript [
   &ParticleState()
   &UpdateMeshOrientation(
    Rotation Rate=&FloatFromCurve(
     FloatCurve=DataInterface() {propertyValues {Curve {keys [{tangentMode=RCTM_User, time=0, value=-1}, <at 0.25 1>, <at 0.5 -1>, <at 0.75 1>, {tangentMode=RCTM_User, time=1, value=-1}]}}}
     CurveIndex=&ReturnNormalizedExecIndex(Normalized Index Includes One=false)
     Scale Curve=&SelectFloatFromArray(Direct Array Index=$Particles.SpawnGroup, Float Selection Array=DataInterface() {propertyValues {FloatData [-1,1,-1,1]}})
    )
    Rotation Vector=<0 0 1>
   )
   &ScaleMeshSize(Scale Factor=&Multiply_VectorByFloat(
     Vector=<1 1 1>
     Float=&FloatFromCurve(
      FloatCurve=DataInterface() {propertyValues {Curve {keys [
          {arriveTangent=7.2197, interpMode=RCIM_Cubic, leaveTangent=7.2197, tangentMode=RCTM_User, time=0, value=0}
          {arriveTangent=1.9952, interpMode=RCIM_Cubic, leaveTangent=1.9952, time=0.4917, value=1.9432}
          {arriveTangent=-0.0405, interpMode=RCIM_Cubic, leaveTangent=-0.0405, tangentMode=RCTM_User, time=1.0024, value=2}
         ]}}}
      CurveIndex=&Divide_Float(
       A=$Particles.NormalizedAge
       B=&SelectFloatFromArray(Direct Array Index=$Particles.SpawnGroup, Float Selection Array=DataInterface() {propertyValues {FloatData [0.9,1.25,1,1.15]}})
      )
      Scale Curve=&SelectFloatFromArray(Direct Array Index=$Particles.SpawnGroup, Float Selection Array=DataInterface() {propertyValues {FloatData [1,0.5,0.9,0.8]}})
     )
    ))
   &ScaleMeshSize(Scale Factor=&MakeVector(
     X=1
     Y=1
     Z=&SelectFloatFromArray(Direct Array Index=$Particles.SpawnGroup, Float Selection Array=DataInterface() {propertyValues {FloatData [2.5,1.5,1.75,2]}})
    )) {module=ScaleMeshSize001}
   &ScaleColor(
    Scale Alpha=&FloatFromCurve(FloatCurve=DataInterface() {propertyValues {Curve {keys [
         {arriveTangent=6.2301, interpMode=RCIM_Cubic, leaveTangent=6.2301, tangentMode=RCTM_User, time=0, value=0.5}
         {arriveTangent=-9.9528, interpMode=RCIM_Cubic, leaveTangent=-9.9528, tangentMode=RCTM_User, time=0.1, value=2}
         <at 0.4 0.2>
         {arriveTangent=0.4048, interpMode=RCIM_Cubic, leaveTangent=0.4048, tangentMode=RCTM_User, time=0.7, value=0}
        ]}}})
    ScaleA=true
    ScaleRGB=true
   )
   &InheritSourceMovement(ApplyForce=false, ApplyPosition=true, ApplyPositionFalloff=false)
   &SolveForcesAndVelocity(Clamp Velocity=false, Limit Acceleration=false)
   &self:ColorFromTOD_02(
    Evening:LinearColor=<rgba 1 0.625 0.5 1>
    Morning:LinearColor=<rgba 1 0.75 0.7753 1>
    Night:LinearColor=<rgba 0.7 0.2966 0.3384 0.75>
    TOD:Vector4f=&MakeVector4_FromLinearColor(LinearColor=$NPC.FortniteNPC.FortniteActiveTimeOfDay)
   ) {module=ColorFromTOD_02}
   &DynamicMaterialParameters(Module.Index0Param0Write=false, Module.Index0Param1Write=true, Module.Index0Param2Write=false, Module.Index0Param3Write=true)
  ]
  rendererClasses [&NiagaraMeshRendererProperties]
  renderers [{rendererClass=&NiagaraMeshRendererProperties, rendererIndex=0}]
 }
}
summary {
 systemName=NS_GrappleKnife_Trail
 emitters [
  {emitterName=Main_Soft_Trail, rendererClasses [&NiagaraRibbonRendererProperties]}
  {emitterName=Wind_Streaks_Circle, rendererClasses [&NiagaraMeshRendererProperties]}
  {emitterName=Facing_Ribbon_Width, rendererClasses [&NiagaraRibbonRendererProperties]}
  {emitterName=InnerFlash, rendererClasses [&NiagaraSpriteRendererProperties]}
 ]
 userVariables [
  User.ColorInner {type=LinearColor, description="", defaultValue {struct=LinearColor, value=<rgba 1 0.8623 0.8424 1>}}
  User.ColorOuter {type=LinearColor, description="", defaultValue {struct=LinearColor, value=<rgba 1 0.9934 0.8035 1>}}
  User.is360Atk {type=NiagaraBool}
  User.LifetimeScale {type=NiagaraFloat, defaultValue=1, description=""}
  User.ParentVelocity {type=Vector3f, description="", defaultValue {struct=Vector3f, value=<0 0 0>}}
  User.RibbonIDOffset {type=NiagaraInt32}
  User.ScaleOpacity {type=NiagaraFloat, defaultValue=5, description=""}
  User.ScaleSize {type=NiagaraFloat, defaultValue=5, description=""}
  User.SetColor {type=NiagaraBool, defaultValue=true}
 ]
}
