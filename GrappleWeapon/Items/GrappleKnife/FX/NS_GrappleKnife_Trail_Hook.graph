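# Niagara trail effect for the grapple knife hook, pairing a ribbon trail with a mesh particle.
# The ribbon emitter spawns continuously, drifting short-lived white segments with upward force and curl noise while alpha rises quickly, then fades.
# Despite its Light name, FN_Projectile_Light001 bursts one stretched mesh particle, with an alpha fade curve and kill-on-lifetime disabled.
# A RibbonIDOffset user parameter is declared, with no visible use.
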NiagaraSystem /GrappleWeapon/Items/GrappleKnife/FX/NS_GrappleKnife_Trail_Hook
dependencies {
 usedDataInterfaces [&NiagaraDataInterfaceCurve, &NiagaraDataInterfaceColorCurve]
 usedDynamicInputs [
  &UniformRangedFloat
  &SimulationPosition
  &MakeVectorFromLinearColorRGB
 ]
 usedModules [&SystemState]
}
emitters {
 FN_Projectile_Light001 {
  emitterUpdateScript [
   &EmitterState(
    Inactive Response=::NewEnumerator1
    Life Cycle Mode=::NewEnumerator1
    Loop Behavior=::NewEnumerator1
    Loop Duration Mode=::NewEnumerator1
    UseLoopDelay=false
   )
   &SpawnBurst_Instantaneous(Spawn Count=1, Use Spawn Probability=false)
  ]
  particleSpawnScript [&InitializeParticle(
    Lifetime=0.25
    Mesh Scale=<4 2 2>
    Mesh Scale Mode=::NewEnumerator4
    Sprite Size=<90 90>
    Sprite Size Mode=::NewEnumerator4
    UsePositionOffset=false
    Write Material Random=true
   ) {module=InitializeParticle001}]
  particleUpdateScript [
   &ParticleState(Kill Particles When Lifetime Has Elapsed=false, Let Infinitely Lived Particles Die When Emitter Deactivates=true)
   &SolveForcesAndVelocity(Clamp Velocity=false, Limit Acceleration=false)
   &ScaleColor(
    Linear Color Curve=DataInterface() {propertyValues {
      AlphaCurve {keys [<at 0 0.5>,<at 1 0>]}
      BlueCurve {keys [<at 0 1>,<at 1 1>]}
      GreenCurve {keys [<at 0 1>,<at 1 1>]}
      RedCurve {keys [<at 0 1>,<at 1 1>]}
     }}
    Scale Mode=::NewEnumerator2
   )
  ]
  rendererClasses [&NiagaraMeshRendererProperties]
  renderers [{rendererClass=&NiagaraMeshRendererProperties, rendererIndex=0}]
 }
 FN_Projectile_Ribbon_Fire {
  emitterUpdateScript [
   &EmitterState()
   &SpawnRate(SpawnRate=30, Use Spawn Probability=false)
  ]
  particleSpawnScript [&InitializeRibbon(
    Lifetime=0.5
    Module.Write Color=true
    Module.Write Lifetime=true
    Module.Write Mass=true
    Module.Write Material Random=true
    Module.Write Position=false
    Module.Write RibbonTwist=true
    Module.Write RibbonWidth=true
    Ribbon Width=80
   )]
  particleUpdateScript [
   &ParticleState()
   &ScaleSpriteSize(Scale Factor=&Vector2DFromFloat(Value=&FloatFromCurve(FloatCurve=DataInterface() {propertyValues {Curve {keys [<at 0.0002 0>,<at 0.2188 0.5881>,<at 0.3557 0.5788>,<at 0.9878 0.35>]}}})))
   &GravityForce(Gravity=<0 0 50>)
   /Niagara/Modules/Update/Forces/CurlNoiseForce(Noise Frequency=0.03, Noise Strength=500, Pan Noise=false)
   &SolveForcesAndVelocity(Clamp Velocity=false, Limit Acceleration=false)
   &Color(Color=&ColorFromCurve(ColorCurve=DataInterface() {propertyValues {
       AlphaCurve {keys [<at 0.0029 0>,<at 0.0609 1 RCIM_Cubic>,<at 1 0 RCIM_Cubic>]}
       BlueCurve {keys [<at 0 1 RCIM_Cubic>]}
       GreenCurve {keys [<at 0 1 RCIM_Cubic>]}
       RedCurve {keys [<at 0 1 RCIM_Cubic>]}
      }}))
  ]
  rendererClasses [&NiagaraRibbonRendererProperties]
  renderers [{rendererClass=&NiagaraRibbonRendererProperties}]
 }
}
summary {
 systemName=NS_GrappleKnife_Trail_Hook
 emitters [
  {emitterName=FN_Projectile_Ribbon_Fire, rendererClasses [&NiagaraRibbonRendererProperties]}
  {emitterName=FN_Projectile_Light001, rendererClasses [&NiagaraMeshRendererProperties]}
 ]
 userVariables [User.RibbonIDOffset {type=NiagaraInt32}]
}
