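# Niagara system for the grapple knife's grappled wind effect, layering mesh gusts and streaking speed lines.
# Air and Wind burst oriented meshes with curve-driven fades, and Wind also scales up over particle and loop age.
# GPU Speedlines sprites combine a burst with a quickly decaying spawn rate, flying from an offset point.
# All three emitters tint by time of day through a ColorFromTOD_03 module reading a Niagara Parameter Collection.
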NiagaraSystem /GrappleWeapon/Items/GrappleKnife/FX/NS_Wind_Grappled
dependencies {
 usedDataInterfaces [&NiagaraDataInterfaceCurve, &NiagaraDataInterfaceVector2DCurve]
 usedDynamicInputs [&Multiply_VectorByFloat]
 usedModules [&SystemState]
}
emitters {
 Air {
  emitterUpdateScript [
   &EmitterState()
   &SpawnBurst_Instantaneous(Use Spawn Probability=false, UseLoopCountLimit=false)
  ]
  particleSpawnScript [
   &InitializeParticle(
    Color=<rgba 1 1 1 0.05>
    Mesh Scale=<1.5 3 3>
    Mesh Scale Mode=::NewEnumerator4
    Position Offset=<0 0 150>
    UsePositionOffset=true
    Write Material Random=true
   )
   &InitialMeshOrientation(Mesh Axis To Orient=<0 -1 0>, Mesh Orientation Mode=::NewEnumerator2, Use Rotation Vector=true)
  ]
  particleUpdateScript [
   &ParticleState()
   &ScaleColor(
    Scale Alpha=&FloatFromCurve(FloatCurve=DataInterface() {propertyValues {Curve {keys [
         {arriveTangent=5.8909, interpMode=RCIM_Cubic, leaveTangent=5.8909, tangentMode=RCTM_User, time=0, value=0}
         {arriveTangent=-3.1448, interpMode=RCIM_Cubic, leaveTangent=-3.1448, tangentMode=RCTM_User, time=0.3, value=1}
         {arriveTangent=0.1065, interpMode=RCIM_Cubic, leaveTangent=0.1065, tangentMode=RCTM_User, time=1, value=0}
        ]}}})
    ScaleA=true
    ScaleRGB=false
   )
   &ScaleColor(
    Scale Alpha=0.75
    Scale RGB=&VectorFromFloat(Value=&FNTimeOfDayScalar(Night Scale=0.35, Morning Scale=0.5, Day Scale=0.8, Evening Scale=0.4))
    ScaleA=true
    ScaleRGB=true
   ) {module=ScaleColor002}
   &self:ColorFromTOD_03(
    Evening:LinearColor=<rgba 1 0.625 0.5 1>
    Morning:LinearColor=<rgba 1 0.75 0.7753 1>
    Night:LinearColor=<rgba 0.7 0.3494 0.4483 1>
    TOD:Vector4f=&MakeVector4_FromLinearColor(LinearColor=$NPC.FortniteNPC.FortniteActiveTimeOfDay)
   ) {module=ColorFromTOD_03}
  ]
  rendererClasses [&NiagaraMeshRendererProperties]
  renderers [{rendererClass=&NiagaraMeshRendererProperties, rendererIndex=0}]
 }
 Speedlines {
  simTarget=GPUComputeSim
  emitterUpdateScript [
   &EmitterState()
   &SpawnRate(
    SpawnRate=&FloatFromCurve(
     FloatCurve=DataInterface() {propertyValues {Curve {keys [
         {arriveTangent=-10.3092, interpMode=RCIM_Cubic, leaveTangent=-10.3092, tangentMode=RCTM_User, time=0, value=1}
         {interpMode=RCIM_Cubic, tangentMode=RCTM_User, time=0.25, value=0}
        ]}}}
     CurveIndex=$System.NormalizedLoopAge
     Scale Curve=400
    )
    Use Spawn Probability=false
   )
   &SpawnBurst_Instantaneous(Spawn Count=24, Use Spawn Probability=false, UseLoopCountLimit=false)
  ]
  particleSpawnScript [
   /Niagara/Modules/Spawn/Initialization/InitializeParticle(
    Color=&LerpLinearColors(
     StartColor=<rgba 1 1 1 0.2>
     EndColor=<rgba 0 0 0 0.4>
     LerpFactor=&UniformAOrBFloat(A=0, Distribution Weight=0.15)
    )
    Lifetime=&UniformRangedFloat(Minimum=0.1, Maximum=0.15)
    Module.Write Color=true
    Module.Write Lifetime=true
    Module.Write Mass=true
    Module.Write Material Random=true
    Module.Write Position=false
    Module.Write Scale=false
    Module.Write SpriteRotation=false
    Module.Write SpriteSize=true
    Position=&SimulationPosition()
    Sprite Size=&UniformRangedVector2D(Minimum=<3 400>, Maximum=<1 500>)
   )
   &ShapeLocation(
    Handle Radius=0
    Large Radius=&RandomRangeFloat(Minimum=17, Maximum=23)
    MaskSpawnGroup=false
    Non Uniform Scale=<1.7 1 1>
    Offset=<0 0 100>
    Offset Mode=::NewEnumerator0
    Override Seed=false
    Rotation Axis=<0 -1 0>
    Rotation Mode=::NewEnumerator1
    Shape Primitive=::NewEnumerator3
    UseOwnerScale=true
   )
   &AddVelocityFromPoint(
    Module.Use Falloff Distance=false
    Origin Offset=<0 50 100>
    Velocity Strength=&UniformRangedFloat(Minimum=1000, Maximum=300)
   )
  ]
  particleUpdateScript [
   &ParticleState()
   &SolveForcesAndVelocity(Clamp Velocity=false, Limit Acceleration=false)
   &ScaleSpriteSize(Scale Factor=&Vector2DFromCurve(
     Vector2Curve=DataInterface() {propertyValues {XCurve {keys [<at 0 0.5 RCIM_Cubic>,<at 0.5 1 RCIM_Cubic>]}, YCurve {keys [<at 0 1 RCIM_Cubic>,<at 0.5 1 RCIM_Cubic>]}}}
     ScaleCurve=&Vector2DFromFloat(Value=1)
    ))
   &ScaleColor(
    Scale Alpha=&Multiply_Float(
     A=&FloatFromCurve(FloatCurve=DataInterface() {propertyValues {Curve {keys [
          {arriveTangent=5.2641, interpMode=RCIM_Cubic, leaveTangent=5.2641, tangentMode=RCTM_User, time=0, value=0}
          {arriveTangent=0.7132, interpMode=RCIM_Cubic, leaveTangent=0.7132, tangentMode=RCTM_User, time=0.3, value=1.0207}
          {arriveTangent=-2.9069, interpMode=RCIM_Cubic, leaveTangent=-2.9069, tangentMode=RCTM_User, time=1, value=0}
         ]}}})
     B=0.1
    )
    ScaleA=true
    ScaleRGB=true
   )
   &self:ColorFromTOD_03(
    Afternoon:LinearColor=<rgba 1 1 1 0.5>
    Evening:LinearColor=<rgba 1 0.625 0.5 0.5>
    Morning:LinearColor=<rgba 1 0.75 0.7753 0.5>
    Night:LinearColor=<rgba 0.7 0.3494 0.4483 0.5>
    TOD:Vector4f=&MakeVector4_FromLinearColor(LinearColor=$NPC.FortniteNPC.FortniteActiveTimeOfDay)
   ) {module=ColorFromTOD_03}
  ]
  rendererClasses [&NiagaraSpriteRendererProperties]
  renderers [{rendererClass=&NiagaraSpriteRendererProperties}]
 }
 Wind {
  emitterUpdateScript [
   &EmitterState(Life Cycle Mode=::NewEnumerator1, Loop Behavior=::NewEnumerator1, Loop Duration=2.4, UseLoopDelay=false)
   &SpawnBurst_Instantaneous(Use Spawn Probability=false, UseLoopCountLimit=false)
  ]
  particleSpawnScript [
   /Niagara/Modules/Spawn/Initialization/V2/InitializeParticle(
    Color Maximum=<rgba 0.6094 0.5078 0.457 0.75>
    Color Minimum=<rgba 0.4167 0.3299 0.3125 0.75>
    Color Mode=::NewEnumerator2
    Lifetime=0.5
    Mesh Scale Mode=::NewEnumerator3
    Mesh Uniform Scale=1.5
    Position Offset=<0 100 100>
    UsePositionOffset=true
    Write Material Random=true
   )
   &InitialMeshOrientation(Mesh Axis To Orient=<0 -1 0>, Mesh Orientation Mode=::NewEnumerator2, Use Rotation Vector=true)
  ]
  particleUpdateScript [
   &ParticleState()
   &ScaleMeshSize(Scale Factor=&MakeVector(
     X=&FloatFromCurve(FloatCurve=DataInterface() {propertyValues {Curve {keys [
          {arriveTangent=20.73, interpMode=RCIM_Cubic, leaveTangent=20.73, tangentMode=RCTM_User, time=0, value=0}
          {arriveTangent=2.2158, interpMode=RCIM_Cubic, leaveTangent=2.2158, tangentMode=RCTM_User, time=0.1, value=0.7}
          {arriveTangent=0.0224, interpMode=RCIM_Cubic, leaveTangent=0.0224, tangentMode=RCTM_User, time=0.5, value=1}
          {arriveTangentWeight=0.1692, interpMode=RCIM_Cubic, tangentMode=RCTM_User, time=1, value=1}
         ]}}})
     Y=1.5
     Z=1.5
    ))
   &ScaleMeshSize(Scale Factor=&VectorFromFloat(Value=&FloatFromCurve(
      FloatCurve=DataInterface() {propertyValues {Curve {keys [
          {arriveTangent=5.0831, interpMode=RCIM_Cubic, leaveTangent=5.0831, tangentMode=RCTM_User, time=0, value=0}
          {interpMode=RCIM_Cubic, tangentMode=RCTM_User, time=0.6, value=1}
         ]}}}
      CurveIndex=$Emitter.NormalizedLoopAge
      Scale Curve=&RandomRangeFloat(Minimum=0.75, Maximum=1.5)
     ))) {module=ScaleMeshSize001}
   &ScaleColor(
    Scale Alpha=&FloatFromCurve(
     FloatCurve=DataInterface() {propertyValues {Curve {keys [
         {arriveTangent=20.1733, interpMode=RCIM_Cubic, leaveTangent=20.1733, tangentMode=RCTM_User, time=0, value=0}
         {
          arriveTangent=1.9289
          arriveTangentWeight=0.1253
          interpMode=RCIM_Cubic
          leaveTangent=1.9289
          leaveTangentWeight=0.2429
          tangentMode=RCTM_User
          time=0.1
          value=0.75
         }
         {arriveTangent=0.0224, interpMode=RCIM_Cubic, leaveTangent=0.0224, tangentMode=RCTM_User, time=0.3, value=1}
         {arriveTangent=-3.0357, interpMode=RCIM_Cubic, leaveTangent=-3.0357, tangentMode=RCTM_User, time=1, value=0}
        ]}}}
     CurveIndex=$Emitter.NormalizedLoopAge
     Scale Curve=4
    )
    ScaleA=true
    ScaleRGB=false
   )
   &self:ColorFromTOD_03(
    Evening:LinearColor=<rgba 1 0.625 0.5 1>
    Morning:LinearColor=<rgba 1 0.75 0.7753 1>
    Night:LinearColor=<rgba 0.7 0.3494 0.4483 1>
    TOD:Vector4f=&MakeVector4_FromLinearColor(LinearColor=$NPC.FortniteNPC.FortniteActiveTimeOfDay)
   ) {module=ColorFromTOD_03}
  ]
  rendererClasses [&NiagaraMeshRendererProperties]
  renderers [{rendererClass=&NiagaraMeshRendererProperties, rendererIndex=0}]
 }
}
summary {
 systemName=NS_Wind_Grappled
 emitters [
  {emitterName=Wind, rendererClasses [&NiagaraMeshRendererProperties]}
  {emitterName=Speedlines, simTarget=GPUComputeSim, rendererClasses [&NiagaraSpriteRendererProperties]}
  {emitterName=Air, rendererClasses [&NiagaraMeshRendererProperties]}
 ]
 userVariables [
  User.Jumped {type=NiagaraBool}
  User.Speed_Alpha {type=NiagaraFloat, defaultValue=0.6, description=""}
  User.uDuration {type=NiagaraFloat, defaultValue=0, description=""}
 ]
}
